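# Translucent UI material for a Discovery browser logo, taking its image from a Logo texture parameter that defaults to an empty texture.
# MF_UI_UVScaling fits the logo to the UI area using its texture size, pivoting at left-center.
# The fitted UVs drive a clamped texture sample.
# MF_UI_ClipEdgePixel takes those UVs and the sampled alpha, and its result is multiplied by that alpha to mask clamped edge pixels.
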Material /DiscoveryBrowser/Art/Materials/Logo/M_UI_Discovery_Logo
graph {
 properties {
  bAutomaticallySetUsageInEditor=false
  bUsedWithStaticMesh=false
  blendMode=BLEND_Translucent
  materialDomain=MD_UI
  usedShadingModels=Default Lit
 }
 nodes [
  n1 = MaterialFunctionCall(
   UV=MaterialFunctionCall() {materialFunction=&GetUserInterfaceUV}.Normalized UV
   TextureSize=TextureProperty(Tex=NamedRerouteUsage())
   ScalePivotPoint=Constant2Vector() {g=0.5}
  ) {materialFunction=&MF_UI_UVScaling}
  n2 = TextureSample(
   UVs=@n1.Scaled to Fit
   Tex=NamedRerouteDeclaration(None=TextureObjectParameter() {parameterName=Logo, texture=/Game/UI/DiscoveryShared/Art/T_UI_Discovery_DefaultEmpty}) {name=Texture Object, nodeColor=<rgba 0 0 0 1>}
  ) {samplerSource=SSM_Clamp_WorldGroupSettings}
  Multiply(
   A=MaterialFunctionCall!(UVs=@n1.Scaled to Fit, Alpha=@n2.A) {materialFunction=/FortUILibrary/MaterialFunctions/Basics/MF_UI_ClipEdgePixel}.Result
   B=@n2.A
  )
 ]
}
